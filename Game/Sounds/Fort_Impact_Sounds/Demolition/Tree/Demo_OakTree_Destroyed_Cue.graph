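# Sound cue for an oak tree being destroyed, in the Fort demolition impact sounds.
# A mixer layers three random picks: leaves at the highest input volume, debris at a middle level and a wood sound at the lowest.
# Each layer passes through its own modulator for slight pitch variation.
# It uses large-object destruction attenuation, the Fort_Impacts sound class and a shared object-impact concurrency set.
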SoundCue /Game/Sounds/Fort_Impact_Sounds/Demolition/Tree/Demo_OakTree_Destroyed_Cue
root=Mixer_0
nodes [
 {id=Mixer_0, children [Modulator_2,Modulator_3,Modulator_5], properties {inputVolume [2.5,1.5,0.5]}}
 {id=Modulator_2, children [Random_0], properties {pitchMax=1.1, pitchMin=0.9}}
 {id=Modulator_3, children [Random_1], properties {pitchMax=1.1, pitchMin=0.9}}
 {id=Modulator_5, children [Random_2], properties {pitchMax=1.1, pitchMin=0.9}}
 {
  id=Random_0
  children [WavePlayer_1,WavePlayer_2,WavePlayer_3,WavePlayer_4]
  properties {bShouldExcludeFromBranchCulling=true, weights [1,1,1,1]}
 }
 {
  id=Random_1
  children [WavePlayer_5, WavePlayer_6, WavePlayer_7, WavePlayer_8, WavePlayer_9, WavePlayer_10, WavePlayer_11, WavePlayer_12]
  properties {bShouldExcludeFromBranchCulling=true, weights [1,1,1,1,1,1,1,1]}
 }
 {id=Random_2, children [WavePlayer_13, WavePlayer_14], properties {bShouldExcludeFromBranchCulling=true, weights [1,1]}}
 {id=WavePlayer_1, properties {soundWaveAssetPtr=&/Demo_Tree_Default_Destroyed_Leaves_04}}
 {id=WavePlayer_10, properties {soundWaveAssetPtr=&/Demo_Tree_Default_Destroyed_Debris_06}}
 {id=WavePlayer_11, properties {soundWaveAssetPtr=&/Demo_Tree_Default_Destroyed_Debris_07}}
 {id=WavePlayer_12, properties {soundWaveAssetPtr=&/Demo_Tree_Default_Destroyed_Debris_08}}
 {id=WavePlayer_13, properties {soundWaveAssetPtr=/Game/Sounds/Fort_Impact_Sounds/Demolition/Wood/final_wood_new_03}}
 {id=WavePlayer_14, properties {soundWaveAssetPtr=/Game/Sounds/Fort_Impact_Sounds/Demolition/Wood/final_wood_new_02}}
 {id=WavePlayer_2, properties {soundWaveAssetPtr=&/Demo_Tree_Default_Destroyed_Leaves_01}}
 {id=WavePlayer_3, properties {soundWaveAssetPtr=&/Demo_Tree_Default_Destroyed_Leaves_02}}
 {id=WavePlayer_4, properties {soundWaveAssetPtr=&/Demo_Tree_Default_Destroyed_Leaves_03}}
 {id=WavePlayer_5, properties {soundWaveAssetPtr=&/Demo_Tree_Default_Destroyed_Debris_01}}
 {id=WavePlayer_6, properties {soundWaveAssetPtr=&/Demo_Tree_Default_Destroyed_Debris_02}}
 {id=WavePlayer_7, properties {soundWaveAssetPtr=&/Demo_Tree_Default_Destroyed_Debris_03}}
 {id=WavePlayer_8, properties {soundWaveAssetPtr=&/Demo_Tree_Default_Destroyed_Debris_04}}
 {id=WavePlayer_9, properties {soundWaveAssetPtr=&/Demo_Tree_Default_Destroyed_Debris_05}}
]
properties {
 attenuationSettings=/Game/Sounds/Attenuation/Fort_Demolition_Destroy_Large_Sized_Object_Att
 maxDistance=6250
 soundClassObject=&Fort_Impacts
 volumeMultiplier=0.8
 concurrencyOverrides {resolutionRule=PreventNew}
 concurrencySet [&CONC_Object_Impacts_FourMax]
}
