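# Sound cue for a palm tree being destroyed, layering randomized leaves, debris, wood and coconut impacts.
# A mixer combines four random-pick layers, each pitch-modulated, with the wood layer shifted higher.
# The coconut layer is boosted most and starts after a short random delay so it lands behind the initial break.
# It uses large-object destruction attenuation, the impacts sound class and a concurrency set for object impacts.
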SoundCue /Game/Sounds/Fort_Impact_Sounds/Demolition/Tree/Demo_PalmTree_Destroyed_Cue
root=Mixer_0
nodes [
 {id=Delay_0, children [Modulator_7], properties {delayMax=0.2, delayMin=0.1}}
 {id=Mixer_0, children [Modulator_1,Modulator_3,Modulator_6,Delay_0], properties {inputVolume [1.5,1.5,1,2.5]}}
 {id=Modulator_1, children [Random_2], properties {pitchMax=1.1, pitchMin=0.9}}
 {id=Modulator_3, children [Random_3], properties {pitchMax=1.1, pitchMin=0.9}}
 {id=Modulator_6, children [Random_4], properties {pitchMax=1.5, pitchMin=1.3}}
 {id=Modulator_7, children [Random_5], properties {pitchMax=1.1, pitchMin=0.9}}
 {
  id=Random_2
  children [WavePlayer_7, WavePlayer_8, WavePlayer_9, WavePlayer_10]
  properties {bShouldExcludeFromBranchCulling=true, weights [1,1,1,1]}
 }
 {
  id=Random_3
  children [WavePlayer_11, WavePlayer_12, WavePlayer_13, WavePlayer_14, WavePlayer_15, WavePlayer_16, WavePlayer_17, WavePlayer_18]
  properties {bShouldExcludeFromBranchCulling=true, weights [1,1,1,1,1,1,1,1]}
 }
 {id=Random_4, children [WavePlayer_19, WavePlayer_20], properties {bShouldExcludeFromBranchCulling=true, weights [1,1]}}
 {
  id=Random_5
  children [WavePlayer_21, WavePlayer_22, WavePlayer_23, WavePlayer_24]
  properties {bShouldExcludeFromBranchCulling=true, weights [1,1,1,1]}
 }
 {id=WavePlayer_10, properties {soundWaveAssetPtr=&/Demo_Tree_Default_Destroyed_Leaves_03}}
 {id=WavePlayer_11, properties {soundWaveAssetPtr=&/Demo_Tree_Default_Destroyed_Debris_01}}
 {id=WavePlayer_12, properties {soundWaveAssetPtr=&/Demo_Tree_Default_Destroyed_Debris_02}}
 {id=WavePlayer_13, properties {soundWaveAssetPtr=&/Demo_Tree_Default_Destroyed_Debris_03}}
 {id=WavePlayer_14, properties {soundWaveAssetPtr=&/Demo_Tree_Default_Destroyed_Debris_04}}
 {id=WavePlayer_15, properties {soundWaveAssetPtr=&/Demo_Tree_Default_Destroyed_Debris_05}}
 {id=WavePlayer_16, properties {soundWaveAssetPtr=&/Demo_Tree_Default_Destroyed_Debris_06}}
 {id=WavePlayer_17, properties {soundWaveAssetPtr=&/Demo_Tree_Default_Destroyed_Debris_07}}
 {id=WavePlayer_18, properties {soundWaveAssetPtr=&/Demo_Tree_Default_Destroyed_Debris_08}}
 {id=WavePlayer_19, properties {soundWaveAssetPtr=/Game/Sounds/Fort_Impact_Sounds/Demolition/Wood/final_wood_new_03}}
 {id=WavePlayer_20, properties {soundWaveAssetPtr=/Game/Sounds/Fort_Impact_Sounds/Demolition/Wood/final_wood_new_02}}
 {id=WavePlayer_21, properties {soundWaveAssetPtr=&/Demo_Tree_Palm_Destroyed_Coconuts_01}}
 {id=WavePlayer_22, properties {soundWaveAssetPtr=&/Demo_Tree_Palm_Destroyed_Coconuts_02}}
 {id=WavePlayer_23, properties {soundWaveAssetPtr=&/Demo_Tree_Palm_Destroyed_Coconuts_03}}
 {id=WavePlayer_24, properties {soundWaveAssetPtr=&/Demo_Tree_Palm_Destroyed_Coconuts_04}}
 {id=WavePlayer_7, properties {soundWaveAssetPtr=&/Demo_Tree_Default_Destroyed_Leaves_04}}
 {id=WavePlayer_8, properties {soundWaveAssetPtr=&/Demo_Tree_Default_Destroyed_Leaves_01}}
 {id=WavePlayer_9, properties {soundWaveAssetPtr=&/Demo_Tree_Default_Destroyed_Leaves_02}}
]
properties {
 attenuationSettings=/Game/Sounds/Attenuation/Fort_Demolition_Destroy_Large_Sized_Object_Att
 maxDistance=6250
 soundClassObject=&Fort_Impacts
 volumeMultiplier=0.8
 concurrencyOverrides {resolutionRule=PreventNew}
 concurrencySet [&CONC_Object_Impacts_FourMax]
}
