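# Sound cue for the final impact when a plant is demolished, built from bush impact and explode waves with medium-object destruction attenuation.
# The first two quality slots mix a randomly picked impact with a half-volume explode layer, while the last two play one fixed impact.
# Both paths get slight random pitch modulation over a raised base pitch.
# Concurrency settings reference a four-max object impact set, plus a prevent-new override rule.
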SoundCue /Game/Sounds/Fort_Impact_Sounds/Demolition/Tree/Demo_Plant_ImpactFinal_Cue
root=QualityLevel_0
nodes [
 {id=Mixer_4, children [Random_39, WavePlayer_121], properties {inputVolume [1,0.5]}}
 {id=Modulator_0, children [Mixer_4], properties {pitchMax=1.07, pitchMin=0.92}}
 {id=Modulator_1, children [WavePlayer_0], properties {pitchMax=1.07, pitchMin=0.92}}
 {id=QualityLevel_0, children [Modulator_0,Modulator_0,Modulator_1,Modulator_1]}
 {id=Random_39, children [WavePlayer_119, WavePlayer_120, WavePlayer_122], properties {weights [1,1,1]}}
 {id=WavePlayer_0, properties {soundWaveAssetPtr=/Game/Sounds/Fort_Impact_Sounds/Demolition/Bush/Bush_ImpactFinal05}}
 {id=WavePlayer_119, properties {soundWaveAssetPtr=&Bush_ImpactFinal05}}
 {id=WavePlayer_120, properties {soundWaveAssetPtr=/Game/Sounds/Fort_Impact_Sounds/Demolition/Bush/Bush_ImpactFinal02}}
 {id=WavePlayer_121, properties {soundWaveAssetPtr=/Game/Sounds/Fort_Impact_Sounds/Demolition/Bush/Bush_Explode}}
 {id=WavePlayer_122, properties {soundWaveAssetPtr=/Game/Sounds/Fort_Impact_Sounds/Demolition/Bush/Bush_ImpactFinal04}}
]
properties {
 attenuationSettings=/Game/Sounds/Attenuation/Fort_Demolition_Destroy_Medium_Sized_Object_Att
 maxDistance=4000
 pitchMultiplier=1.5
 soundClassObject=&Fort_Impacts_PSM
 volumeMultiplier=0.34
 concurrencyOverrides {resolutionRule=PreventNew}
 concurrencySet [&CONC_Object_Impacts_FourMax]
}
