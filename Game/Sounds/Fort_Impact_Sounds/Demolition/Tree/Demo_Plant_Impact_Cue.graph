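# Sound cue for plant demolition impacts, built from shared bush impact waves and pitched up overall at reduced volume.
# A quality-level node picks among four pitch-modulated branches, mostly shrinking the wave pool from three random variants down to a single wave.
# The second branch's random node keeps an empty third input.
# Playback uses small-object demolition attenuation, the impacts sound class, and a four-max concurrency set alongside a prevent-new override.
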SoundCue /Game/Sounds/Fort_Impact_Sounds/Demolition/Tree/Demo_Plant_Impact_Cue
root=QualityLevel_0
nodes [
 {id=Modulator_0, children [Random_0], properties {pitchMax=1.07, pitchMin=0.92}}
 {id=Modulator_1, children [Random_1], properties {pitchMax=1.07, pitchMin=0.92}}
 {id=Modulator_12, children [Random_38], properties {pitchMax=1.07, pitchMin=0.92}}
 {id=Modulator_2, children [WavePlayer_4], properties {pitchMax=1.1, pitchMin=0.9}}
 {id=QualityLevel_0, children [Modulator_12,Modulator_0,Modulator_1,Modulator_2]}
 {id=Random_0, children [WavePlayer_1,WavePlayer_0,null], properties {weights [1,1,1]}}
 {id=Random_1, children [WavePlayer_3,WavePlayer_2], properties {weights [1,1]}}
 {id=Random_38, children [WavePlayer_118, WavePlayer_117, WavePlayer_116], properties {weights [1,1,1]}}
 {id=WavePlayer_0, properties {soundWaveAssetPtr=/Game/Sounds/Fort_Impact_Sounds/Demolition/Bush/Bush_Impact01}}
 {id=WavePlayer_1, properties {soundWaveAssetPtr=/Game/Sounds/Fort_Impact_Sounds/Demolition/Bush/Bush_Impact10}}
 {id=WavePlayer_116, properties {soundWaveAssetPtr=/Game/Sounds/Fort_Impact_Sounds/Demolition/Bush/Bush_Impact06}}
 {id=WavePlayer_117, properties {soundWaveAssetPtr=&Bush_Impact01}}
 {id=WavePlayer_118, properties {soundWaveAssetPtr=&Bush_Impact10}}
 {id=WavePlayer_2, properties {soundWaveAssetPtr=&Bush_Impact01}}
 {id=WavePlayer_3, properties {soundWaveAssetPtr=&Bush_Impact10}}
 {id=WavePlayer_4, properties {soundWaveAssetPtr=&Bush_Impact10}}
]
properties {
 attenuationSettings=/Game/Sounds/Attenuation/Fort_Demolition_Hit_Small_Sized_Object_Att
 maxDistance=4000
 pitchMultiplier=1.5
 soundClassObject=&Fort_Impacts
 volumeMultiplier=0.34
 concurrencyOverrides {resolutionRule=PreventNew}
 concurrencySet [&CONC_Object_Impacts_FourMax]
}
